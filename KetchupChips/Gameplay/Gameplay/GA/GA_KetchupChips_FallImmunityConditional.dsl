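; KetchupChips gameplay ability, a conditional fall-immunity variant tied to a sprint-jump ability.
; Extends parent setup with two tag-removal waits, covering sprint-jump cancel and primary-ability cancellation (e.g. God Fruit duration), both calling OnSprintEnd.
; OnSprintEnd flags the jump-sprint as ended if the pawn is falling and ends the ability; CallEnd defers to the parent only when that flag is set.
; Stored wait tasks are ended on EndAbility.
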(event EventSetUpWaits
  (|CallParentFunction)
  (bind _asynctaskproxy (Ability|Tasks|WaitGameplayTagRemove (Variables|Default|GetToCancelSprintJumpAbilityTag))
    (:then
      (Utilities|Array|Add (Variables|Default|GetWaitTagRemovedAsyncs) _asynctaskproxy) ; store for cleanup
      (bind _asynctaskproxy_2 (Ability|Tasks|WaitGameplayTagRemove (Variables|Default|GetTagPrimaryAbilityCancelled))
        (:then
          (Utilities|Array|Add (Variables|Default|GetWaitTagRemovedAsyncs) _asynctaskproxy_2))
        (:Removed
          (CallFunction|OnSprintEnd))) ; if the ability is cancelled some other way\n(e.g. God Fruit duration)
    )
    (:Removed
      (CallFunction|OnSprintEnd))))

(event Custom|OnSprintEnd
  (bind _asfort_pawn (Utilities|Casting|CastToFortPawn (Ability|GetAvatarActorfromActorInfo))
    (:then
      (if (AI|Components|NavMovement|IsFalling (Variables|/Script/Engine.Character|GetCharacterMovement _asfort_pawn))
        (Variables|Default|SetbIsJumpSprintEnded true)
        (else
          _)))
    (:CastFailed
      _))
  (Ability|EndAbility))

(event EventCallEnd
  (if (Variables|Default|GetbIsJumpSprintEnded)
    (|CallParentFunction)))

(event Ability|EventOnEndAbility
  (params (bWasCancelled :type Boolean))
  (for _array_element (Variables|Default|GetWaitTagRemovedAsyncs)
    (Utilities|IsValid _array_element
      (:Is%20Valid
        (GameplayTasks|EndTask _array_element))))
  (|CallParentFunction bWasCancelled))
